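# Unlit translucent smoke material for Niagara and Cascade sprites in the DudeBro environment, using an inky smoke flipbook texture.
# Texture brightness, shaped by dynamic-parameter exponents and ColorScale, blends low and high colors.
# That tint, multiplied by particle color and texture, feeds an emissive-adjustment function.
# An alpha branch multiplies particle alpha by texture, adding depth fade except on mobile; OpacityExponent affects only color.
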Material /Game/Athena/Environments/DudeBro/M_DudeBro_Smoke_01
graph {
 properties {
  bUsedWithNiagaraSprites=true
  bUsedWithParticleSprites=true
  blendMode=BLEND_Translucent
  pixelDepthOffsetMode=PDOM_Legacy
  shadingModel=MSM_Unlit
  usedShadingModels=Unlit
 }
 nodes [
  n1 = ScalarParameter() {defaultValue=1.5, parameterName=ColorScale}
  n2 = TextureSample() {texture=/Game/Effects/Fort_Effects/Textures/T_InkySmoke_Flipbook}
  n3 = Power(
   Base=@n2.RGB
   Exp=DynamicParameter() {defaultValue=<rgba 10 1 1 1>, paramNames [ColorExponent, OpacityExponent, Param3, Param4]}.OpacityExponent
  )
  n4 = LinearInterpolate(
   A=VectorParameter() {parameterName=Color Low, defaultValue=<rgba 0.0988 0.0123 0.4993 0>}.RGB
   B=VectorParameter() {parameterName=Color High, defaultValue=<rgba 0.6967 0.1012 0.9725 0>}.RGB
   Alpha=Power(
    Base=Multiply(A=@n1, B=@n3)
    Exp=DynamicParameter() {defaultValue=<rgba 10 1 1 1>, paramNames [ColorExponent, OpacityExponent, Param3, Param4]}.ColorExponent
   )
  )
  n5 = ParticleColor()
  MaterialFunctionCall(Emissive=Multiply(A=Multiply(A=@n4, B=@n5.RGB), B=@n2.RGB)) {materialFunction=&MF_FX_EmissiveAdjustments}
  n6 = ScalarParameter() {parameterName=Fade Distance}
  n7 = Multiply(A=@n5.A, B=@n2.RGB)
  ShadingPathSwitch(Default=Multiply(A=DepthFade(FadeDistance=@n6) {fadeDistanceDefault=20}, B=@n7), Mobile=@n7) {desc=Migrated from FeatureLevelSwitch}
 ]
}
