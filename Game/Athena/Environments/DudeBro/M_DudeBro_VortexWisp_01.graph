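# Unlit alpha-composite particle material for a purple vortex wisp effect on mesh or sprite particles.
# Two panning cloud and smoke textures blend low and high colors, with a brightened sweep edge.
# Fresnel, sphere masks and depth fade shape the emissive, tinted by particle color.
# Dynamic parameters drive panning, tiling, sweep and a bend that rotates vertices around particle-local Z, weighted by U and easing out over particle life.
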Material /Game/Athena/Environments/DudeBro/M_DudeBro_VortexWisp_01
graph {
 properties {
  bUsedWithMeshParticles=true
  bUsedWithParticleSprites=true
  blendMode=BLEND_AlphaComposite
  pixelDepthOffsetMode=PDOM_Legacy
  shadingModel=MSM_Unlit
  usedShadingModels=Unlit
 }
 nodes [
  n1 = VectorParameter() {parameterName=ColorHigh, defaultValue=<rgba 0.2739 0.0466 0.425 0>}
  n2 = TextureCoordinate() {uTiling=0.666}
  n3 = MaterialFunctionCall(XY=@n2) {materialFunction=&Swizzle}
  n4 = AppendVector(A=DynamicParameter() {defaultValue=<rgba 1 1 0.5 1>, paramNames [PanSpeedU,BendAmount,Sweep,TileU]}.TileU, B=Constant() {r=1})
  n5 = MaterialFunctionCall(Float2=Multiply(A=StaticSwitchParameter(True=@n3.YX, False=@n2) {parameterName=FlipUVs}, B=@n4)) {materialFunction=&BreakOutFloat2Components}
  n6 = OneMinus(None=@n5.G)
  n7 = AppendVector(A=DynamicParameter() {defaultValue=<rgba 1 1 0.5 1>, paramNames [PanSpeedU,BendAmount,Sweep,TileU]}.PanSpeedU, B=Constant())
  n8 = TextureSample(UVs=Panner(Coordinate=AppendVector(A=@n5.R, B=@n6), Speed=Multiply(A=@n7) {constB=1.15}) {bFractionalPart=true}) {texture=/Game/Effects/Fort_Effects/Textures/Tile/T_CloudScape_08}
  n9 = TextureCoordinate()
  n10 = StaticSwitchParameter(True=MaterialFunctionCall(XY=@n9) {materialFunction=&Swizzle}.YX, False=@n9) {parameterName=FlipUVs}
  n11 = TextureSample(UVs=Panner(Coordinate=Multiply(A=@n10, B=@n4), Speed=@n7) {bFractionalPart=true}) {samplerType=SAMPLERTYPE_Grayscale, texture=&T_Smoke_Trail_Soft}
  n12 = Power(Base=Divide(A=Add(A=@n8.RGB, B=@n11.RGB)) {constB=1.3}, Exp=ScalarParameter() {defaultValue=2.5, parameterName=Exponent})
  n13 = ComponentMask(None=Power(Base=TextureCoordinate()) {constExponent=3}) {r=true}
  n14 = DynamicParameter() {defaultValue=<rgba 1 1 0.5 1>, paramNames [PanSpeedU,BendAmount,Sweep,TileU]}
  n15 = ScalarParameter() {defaultValue=0.4, parameterName=SweepRadius}
  n16 = ScalarParameter() {defaultValue=0.2, parameterName=SweepHarness}
  n17 = Abs(None=Subtract(A=TextureCoordinate()) {constB=0.5})
  n18 = Multiply(
   A=StaticSwitchParameter(True=SphereMask(A=@n13, B=@n14.Sweep, Radius=@n15, Hardness=@n16), False=Constant() {r=1}) {defaultValue=true, parameterName=UseSweep}
   B=SphereMask(
    A=Power(Base=@n17)
    B=Constant()
    Radius=ScalarParameter() {defaultValue=0.25, parameterName=SphereMaskRadius}
    Hardness=ScalarParameter() {defaultValue=0.5, parameterName=SphereMaskHardness}
   )
  )
  n19 = Multiply(A=Power(Base=@n18) {constExponent=9}, B=ScalarParameter() {defaultValue=3, parameterName=LeadingEdgeScale})
  n20 = Add(
   A=LinearInterpolate(A=VectorParameter() {parameterName=ColorLow, defaultValue=<rgba 0.0093 0.0026 0.1198 0>}.RGB, B=@n1.RGB, Alpha=@n12)
   B=Multiply(A=@n1.RGB, B=Multiply(A=@n12, B=@n19))
  )
  n21 = ParticleColor()
  n22 = MaterialFunctionCall(
   Brightness=ScalarParameter() {defaultValue=1.2, parameterName=FrenselBrightness}
   Exponent=ScalarParameter() {defaultValue=4, parameterName=FresnelExponent}
  ) {materialFunction=&MF_Fresnel_FX}
  n23 = DepthFade(
   Opacity=Multiply(A=Multiply(A=@n12, B=@n22.Fresnel), B=@n18)
   FadeDistance=ScalarParameter() {defaultValue=100, parameterName=DepthFadeDistance}
  )
  MaterialFunctionCall(Emissive=Multiply(A=Multiply(A=@n20, B=@n21.RGB), B=Multiply(A=@n21.A, B=@n23))) {materialFunction=&MF_FX_EmissiveAdjustments}
  n24 = Constant3Vector() {constant=<rgba 0 0 1 0>}
  n25 = DynamicParameter() {defaultValue=<rgba 1 1 0.5 1>, paramNames [PanSpeedU,BendAmount,Sweep,TileU]}
  n26 = OneMinus(None=ParticleRelativeTime())
  Multiply(
   A=RotateAboutAxis(
    NormalizedRotationAxis=Transform(None=@n24) {transformSourceType=TRANSFORMSOURCE_Local, transformType=TRANSFORM_ParticleWorld}
    RotationAngle=Multiply(A=@n25.BendAmount, B=@n26)
    PivotPoint=ObjectPositionWS()
    Position=WorldPosition().XYZ
   )
   B=MaterialFunctionCall(Float2=TextureCoordinate()) {materialFunction=&BreakOutFloat2Components}.R
  )
 ]
}
